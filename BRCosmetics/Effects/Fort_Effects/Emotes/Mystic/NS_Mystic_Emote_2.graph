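# Niagara system for the Mystic emote: a single GPU sprite emitter that bursts bright orange sparks from a skeletal mesh.
# Particles spawn at random bone/socket indices on User.SK_mesh, push out from a raised point and inherit velocity.
# Curl noise, drag and gravity shape their short flight, while sprites stretch with speed and fade with a time-of-day alpha multiplier.
# The rate spawn, sphere location and vortex modules are disabled.
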NiagaraSystem /BRCosmetics/Effects/Fort_Effects/Emotes/Mystic/NS_Mystic_Emote_2
dependencies {usedDataInterfaces [&NiagaraDataInterfaceCurve], usedDynamicInputs [], usedModules [&SystemState]}
emitters {sparkes {
  simTarget=GPUComputeSim
  emitterUpdateScript [
   &EmitterState(Inactive Response=::NewEnumerator1, Life Cycle Mode=::NewEnumerator1, Loop Behavior=::NewEnumerator1, Loop Duration=0.65, UseLoopDelay=false)
   &SpawnRate(
    SpawnRate=&FloatFromCurve(
     FloatCurve=DataInterface() {propertyValues {Curve {keys [
         {arriveTangent=3.6874, interpMode=RCIM_Cubic, leaveTangent=3.6874, tangentMode=RCTM_User, time=0.1, value=0.3}
         {tangentMode=RCTM_User, time=1, value=1.8}
        ]}}}
     CurveIndex=$Emitter.Age
     Scale Curve=480
    )
    Use Spawn Probability=false
   ) {enabled=false}
   /SetVariables() {isSetParameters=true}
   &SpawnBurst_Instantaneous(Spawn Count=50, Use Spawn Probability=false, UseLoopCountLimit=false)
  ]
  particleSpawnScript [
   /Niagara/Modules/Spawn/Initialization/InitializeParticle(
    Color=<rgba 25 3.8042 0 1>
    Lifetime=&UniformRangedFloat(Minimum=0.1, Maximum=0.6)
    Module.Write Color=true
    Module.Write Lifetime=true
    Module.Write Mass=false
    Module.Write Material Random=true
    Module.Write Position=false
    Module.Write Scale=false
    Module.Write SpriteRotation=true
    Module.Write SpriteSize=true
    Position=&SimulationPosition()
    Sprite Rotation=&UniformRangedFloat(Minimum=-180, Maximum=180)
    Sprite Size=&Vector2DFromFloat(Value=&UniformRangedFloat(Minimum=1, Maximum=2))
   )
   &SkeletalMeshLocation(
    Bone / Socket Index=&RandomRangeInt(Minimum=1, Maximum=40)
    Bone Sampling Mode=::NewEnumerator5
    MaskSpawnGroup=false
    Sampled Position Offset=&RandomRangeVector(Minimum=<-1 -1 -1>, Maximum=<-1 1 1>)
    Skeletal Mesh=$User.SK_mesh
   )
   /Niagara/Modules/Spawn/Location/SphereLocation(MaskSpawnGroup=false, Module.Override Randomness=false, Non Uniform Scale=<2 1 0.5>, Offset=<0 0 9>, Sphere Radius=5, Surface Only=false) {enabled=false}
   &AddVelocityFromPoint(
    Module.Use Falloff Distance=false
    Origin Offset=<0 0 23>
    Origin Offset Coordinate Space=::Simulation
    Velocity Strength=&RandomRangeFloat(Minimum=20, Maximum=100)
   )
   &AddVelocity(Coordinate Space=::World, Velocity=&RandomRangeVector(Minimum=<0 0 -10>, Maximum=<0 0 10>)) {module=AddVelocity001}
   &InheritVelocity(Speed Limit=true, UseSpeedThreshold=false)
  ]
  particleUpdateScript [
   &ParticleState()
   /Niagara/Modules/Update/Forces/CurlNoiseForce(
    Noise Frequency=0.1
    Noise Strength=&FloatFromCurve(
     FloatCurve=DataInterface() {propertyValues {Curve {keys [<at 0 0>, {arriveTangent=1, interpMode=RCIM_Cubic, leaveTangent=1, time=0.4106, value=0.6448}, <at 1 1>]}}}
     Scale Curve=300
    )
    Pan Noise=true
    Pan Noise Field=<0.5 -1 1>
   )
   &Drag(Drag=&RandomRangeFloat(Minimum=1, Maximum=5), Use Linear Drag=true, Use Rotational Drag=false)
   &VortexForce(UseOriginOffset=false, Vortex Axis=<0 1 0>, Vortex Force Amount=1000) {enabled=false}
   &ScaleColor(
    Scale Alpha=&Multiply_Float(
     A=&FloatFromCurve(FloatCurve=DataInterface() {propertyValues {Curve {keys [
          {arriveTangent=11.9433, interpMode=RCIM_Cubic, leaveTangent=11.9433, tangentMode=RCTM_User, time=0, value=0}
          <at 0.2 1 RCIM_Cubic>
          <at 1 0 RCIM_Cubic>
         ]}}})
     B=&Multiply_Float(A=1, B=&FNTimeOfDayScalar(Night Scale=0.5, Morning Scale=0.25, Evening Scale=0.25))
    )
    ScaleA=true
    ScaleRGB=true
   ) {module=ScaleColor001}
   &DynamicMaterialParameters(
    Index 0 Param 1=&FloatFromCurve(FloatCurve=DataInterface() {propertyValues {Curve {keys [<at 0 0>,<at 1 1>]}}})
    Index 0 Param 2=1
    Index 0 Param 3=1
    Index 0 Param 4=1
    Module.Param0WriteEnabled=true
    Module.Param1WriteEnabled=true
    Module.Param2WriteEnabled=true
    Module.Param3WriteEnabled=true
   )
   &ScaleSpriteSize(Scale Factor=&Vector2DFromFloat(Value=&FloatFromCurve(FloatCurve=DataInterface() {propertyValues {Curve {keys [
          {arriveTangent=9.1531, interpMode=RCIM_Cubic, leaveTangent=9.1531, tangentMode=RCTM_User, time=0, value=0.02}
          {arriveTangent=0.48, interpMode=RCIM_Cubic, leaveTangent=0.48, time=0.18, value=0.99}
          <at 1 0.5 RCIM_Cubic>
         ]}}})))
   &GravityForce(Gravity=<0 0 -200>)
   &SolveForcesAndVelocity(Clamp Velocity=false, Limit Acceleration=false)
   &ScaleSpriteSizeBySpeed(Max Scale Factor=<0.5 20>, Min Scale Factor=<0.3 0.5>)
  ]
  rendererClasses [&NiagaraSpriteRendererProperties]
  renderers [{rendererClass=&NiagaraSpriteRendererProperties}]
 }}
summary {
 systemName=NS_Mystic_Emote_2
 emitters [{emitterName=sparkes, simTarget=GPUComputeSim, rendererClasses [&NiagaraSpriteRendererProperties]}]
 userVariables [
  User.bIsFrontEnd {type=NiagaraBool}
  User.bIsFrontEndPreview {type=NiagaraBool}
  User.RibbonIDOffset {type=NiagaraInt32}
  User.SK_mesh {
   type=NiagaraDataInterfaceSkeletalMesh
   defaultValue {
    struct=NiagaraExt_VariableValue_DataInterface
    value {dataInterface=&self:NiagaraDataInterfaceSkeletalMesh_3}
   }
  }
 ]
}
