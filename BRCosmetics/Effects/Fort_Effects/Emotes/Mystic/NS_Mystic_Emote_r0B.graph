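# Niagara effect for the Mystic emote that draws a glowing orange spiral ring with flying sparks.
# A ribbon spawns along a Y-axis torus whose radius shrinks as Emitter.Age advances the spawn angle, tracing an inward spiral.
# GPU sparks follow a similar shrinking path, driven by vortex, curl noise and drag, with speed-based stretching and time-of-day dimming.
# A center glow sprite and a sprite-plus-light burst add illumination, faded by age curves.
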NiagaraSystem /BRCosmetics/Effects/Fort_Effects/Emotes/Mystic/NS_Mystic_Emote_r0B
dependencies {
 usedDataInterfaces [
  &NiagaraDataInterfaceCurve
  &NiagaraDataInterfaceColorCurve
  &NiagaraDataInterfaceVector2DCurve
  &NiagaraDataInterfaceVectorCurve
 ]
 usedDynamicInputs [&VectorFromCurve]
 usedModules [
  &SystemState
  &self:SystemSpawnScript.SystemScriptSource.SystemScriptGraph.NiagaraNodeAssignment_1.SetVariables_#1
 ]
}
emitters {
 Center_Glow {
  emitterUpdateScript [
   &EmitterState()
   &SpawnBurst_Instantaneous(Use Spawn Probability=false, UseLoopCountLimit=false)
  ]
  particleSpawnScript [&InitializeParticle(
    Color=<rgba 2.6 0.31 0 0.7>
    Lifetime Max=1
    Lifetime Min=0.8
    Lifetime Mode=::NewEnumerator1
    Sprite Size=<500 400>
    Sprite Size Mode=::NewEnumerator4
    UsePositionOffset=false
    Write Material Random=true
   )]
  particleUpdateScript [
   &ParticleState(Kill Particles When Lifetime Has Elapsed=false, Let Infinitely Lived Particles Die When Emitter Deactivates=true)
   &ScaleColor(
    Scale Alpha=&FloatFromCurve(
     FloatCurve=DataInterface() {propertyValues {Curve {keys [
         {arriveTangent=3.6277, interpMode=RCIM_Cubic, leaveTangent=3.6277, tangentMode=RCTM_User, time=0, value=0}
         {interpMode=RCIM_Cubic, tangentMode=RCTM_User, time=0.5, value=1}
        ]}}}
     CurveIndex=$Emitter.Age
     Scale Curve=&FloatFromCurve(
      FloatCurve=DataInterface() {propertyValues {Curve {keys [
          {interpMode=RCIM_Cubic, tangentMode=RCTM_User, time=0.1, value=0}
          {arriveTangent=2.7482, interpMode=RCIM_Cubic, leaveTangent=2.7482, tangentMode=RCTM_User, time=0.3, value=0.4}
          {interpMode=RCIM_Cubic, tangentMode=RCTM_User, time=0.7, value=0.95}
          <at 2 0>
         ]}}}
      CurveIndex=$Emitter.Age
      Scale Curve=0.3
     )
    )
    ScaleA=true
    ScaleRGB=false
   )
   &ScaleSpriteSize(
    Uniform Curve Index=$Emitter.Age
    Uniform Curve Scale=0.29
    Uniform Curve Sprite Scale=DataInterface() {propertyValues {Curve {keys [{interpMode=RCIM_Cubic, tangentMode=RCTM_User, time=1, value=1}, {interpMode=RCIM_Cubic, tangentMode=RCTM_User, time=2, value=0.4}]}}}
   )
   &CameraOffset(Camera Offset Amount=-1)
   &ShapeLocation(MaskSpawnGroup=false, Override Seed=false, Sphere Radius=0, UseOwnerScale=true)
  ]
  rendererClasses [&NiagaraSpriteRendererProperties]
  renderers [{rendererClass=&NiagaraSpriteRendererProperties}]
 }
 FN_Ribbon_Attributes_Test {
  emitterUpdateScript [
   &EmitterState(Inactive Response=::NewEnumerator1, Life Cycle Mode=::NewEnumerator1, Loop Behavior=::NewEnumerator1, Loop Duration=1.5, UseLoopDelay=false) {module=EmitterLifeCycle}
   &SpawnRate(SpawnRate=30, Use Spawn Probability=false)
  ]
  particleSpawnScript [
   /SetVariables(Particles.My_Random:NiagaraFloat=&UniformRangedFloat(Minimum=0, Maximum=1)) {isSetParameters=true}
   /SetVariables(Particles.ParticleId_Thing=HlslExpression() {hlslExpression=Particles.ID}) {isSetParameters=true}
   &InitializeRibbon(
    Lifetime=0.8
    Mass=10
    Module.Write Color=true
    Module.Write Lifetime=true
    Module.Write Mass=true
    Module.Write Material Random=true
    Module.Write Position=false
    Module.Write RibbonTwist=false
    Module.Write RibbonWidth=false
    Position:Vector3f=&SimulationPosition()
    Ribbon Width=1
   )
   &TorusLocation(
    Align Torus With Axis=true
    Large Radius:NiagaraFloat=&FloatFromCurve(
     FloatCurve=DataInterface() {propertyValues {Curve {keys [<at 0.2 40 RCIM_Cubic>, {arriveTangent=-2.0404, interpMode=RCIM_Cubic, leaveTangent=-2.0404, tangentMode=RCTM_User, time=1.25, value=4}]}}}
     CurveIndex=$Emitter.Age
    )
    MaskSpawnGroup=false
    Torus Axis=<0 -1 0>
    Torus Distribution Mode:ENiagaraTorusDistributionMode=::NewEnumerator1
    Torus Mode=::NewEnumerator2
    U Position:NiagaraFloat=&Multiply_Float(A=$Emitter.Age, B=2)
   )
  ]
  particleUpdateScript [
   &ParticleState()
   /SetVariables(Particles.RibbonWidth=&Cosine(
     Normalized Angle=$Emitter.NormalizedLoopAge
     Period=1
     Scale=0
     Bias=&FloatFromCurve(
      FloatCurve=DataInterface() {propertyValues {Curve {keys [<at 0.6 1 RCIM_Cubic>,<at 1.85 0.1 RCIM_Cubic>]}}}
      CurveIndex=$Emitter.Age
      Scale Curve=12.9
     )
    )) {isSetParameters=true}
   &Color(Color=&MultiplyLinearColor(
     Linear Color=&ColorFromCurve(
      ColorCurve=DataInterface() {propertyValues {
        AlphaCurve {keys [
          <at 0 0 RCIM_Cubic>
          {arriveTangent=2.7031, interpMode=RCIM_Cubic, leaveTangent=2.7031, tangentMode=RCTM_User, time=0.6, value=0.375}
          {arriveTangent=-2.2823, interpMode=RCIM_Cubic, leaveTangent=-2.2823, tangentMode=RCTM_User, time=0.85, value=1}
          {interpMode=RCIM_Cubic, tangentMode=RCTM_User, time=1.5, value=0}
         ]}
        BlueCurve {keys [<at 1 0.05 RCIM_Cubic>]}
        GreenCurve {keys [<at 1 8.6103 RCIM_Cubic>]}
        RedCurve {keys [<at 1 90 RCIM_Cubic>,<at 1 90 RCIM_Cubic>]}
       }}
      CurveIndex=$Emitter.Age
     )
     Scale Factor=<1 1 1 1>
    )) {module=Color003}
   &SolveForcesAndVelocity(Clamp Velocity=false, Limit Acceleration=false)
  ]
  rendererClasses [&NiagaraRibbonRendererProperties]
  renderers [{rendererClass=&NiagaraRibbonRendererProperties}]
 }
 SimpleSpriteBurst {
  emitterUpdateScript [
   &EmitterState(Inactive Response=::NewEnumerator1, Life Cycle Mode=::NewEnumerator1, Loop Behavior=::NewEnumerator1, Loop Duration=2.4, UseLoopDelay=false)
   &SpawnBurst_Instantaneous(Use Spawn Probability=false, UseLoopCountLimit=false)
  ]
  particleSpawnScript [&InitializeParticle(Lifetime=2.4, Sprite Size Mode=::NewEnumerator3, Uniform Sprite Size=50, UsePositionOffset=false, Write Material Random=true)]
  particleUpdateScript [
   &ParticleState()
   &ScaleColor(
    Scale Alpha=&FloatFromCurve(FloatCurve=DataInterface() {propertyValues {Curve {keys [<at -0.0046 1.0113 RCIM_Cubic>,<at 1.0011 0.002 RCIM_Cubic>]}}})
    ScaleA=true
    ScaleRGB=false
   )
   &SolveForcesAndVelocity(Clamp Velocity=false, Limit Acceleration=false)
   &FN_Light_Attributes(
    Color=<rgba 1 0.2 0.02 1>
    Create Color Attribute=true
    Create Exponent Attribute=true
    Create Position Attribute=true
    Create Radius Attribute=true
    Exponent=2
    Position=&Add_Vector(
     A=&ConvertPositionToVector(Input Position=$Particles.Position, Convert To Absolute Large World Vector=false)
     B=$System.SystemOfs
    )
    Radius=33
    Scale Color=&FloatFromCurve(
     FloatCurve=DataInterface() {propertyValues {Curve {keys [
         {interpMode=RCIM_Cubic, tangentMode=RCTM_User, time=0.2, value=0}
         {arriveTangent=0.0866, interpMode=RCIM_Cubic, leaveTangent=0.0866, tangentMode=RCTM_User, time=0.6, value=0.8}
         {arriveTangent=-0.1788, interpMode=RCIM_Cubic, leaveTangent=-0.1788, tangentMode=RCTM_User, time=2, value=1}
         <at 2.3 0>
        ]}}}
     CurveIndex=$Emitter.Age
     Scale Curve=7
    )
   )
  ]
  rendererClasses [&NiagaraSpriteRendererProperties, &NiagaraLightRendererProperties]
  renderers [
   {rendererClass=&NiagaraSpriteRendererProperties}
   {rendererClass=&NiagaraLightRendererProperties, rendererIndex=1}
  ]
 }
 sparkes {
  simTarget=GPUComputeSim
  emitterUpdateScript [
   &EmitterState(Life Cycle Mode=::NewEnumerator1, Loop Behavior=::NewEnumerator1, Loop Delay=0.4625, Loop Duration=0.7792, UseLoopDelay=true)
   &SpawnRate(
    SpawnRate=&FloatFromCurve(
     FloatCurve=DataInterface() {propertyValues {Curve {keys [
         {arriveTangent=3.6874, interpMode=RCIM_Cubic, leaveTangent=3.6874, tangentMode=RCTM_User, time=0.1, value=0.05}
         {tangentMode=RCTM_User, time=1, value=1.8}
        ]}}}
     CurveIndex=$Emitter.Age
     Scale Curve=400
    )
    Use Spawn Probability=false
   )
   /SetVariables() {isSetParameters=true}
  ]
  particleSpawnScript [
   /Niagara/Modules/Spawn/Initialization/InitializeParticle(
    Color=<rgba 25 3.8042 0 1>
    Lifetime=&UniformRangedFloat(Minimum=0.1, Maximum=0.3)
    Module.Write Color=true
    Module.Write Lifetime=true
    Module.Write Mass=false
    Module.Write Material Random=true
    Module.Write Position=false
    Module.Write Scale=false
    Module.Write SpriteRotation=true
    Module.Write SpriteSize=true
    Position=&SimulationPosition()
    Sprite Rotation=&UniformRangedFloat(Minimum=-180, Maximum=180)
    Sprite Size=&Vector2DFromFloat(Value=&UniformRangedFloat(Minimum=1, Maximum=2))
   )
   /Niagara/Modules/Spawn/Location/SphereLocation(MaskSpawnGroup=false, Module.Override Randomness=false, Sphere Radius=1, Surface Only=false) {enabled=false}
   &TorusLocation(
    Align Torus With Axis=true
    Large Radius:NiagaraFloat=&FloatFromCurve(
     FloatCurve=DataInterface() {propertyValues {Curve {keys [<at 0.25 -40 RCIM_Cubic>, {arriveTangent=7.6273, interpMode=RCIM_Cubic, leaveTangent=7.6273, tangentMode=RCTM_User, time=1.3, value=-10}]}}}
     CurveIndex=$Emitter.Age
    )
    MaskSpawnGroup=false
    Radius Position:NiagaraFloat=&RandomRangeFloat(Minimum=0.8, Maximum=1)
    Torus Axis=<0 -1 0>
    Torus Distribution Mode:ENiagaraTorusDistributionMode=::NewEnumerator1
    Torus Mode=::NewEnumerator2
    U Position:NiagaraFloat=&Multiply_Float(
     A=&RandomRangeFloat(Minimum=HlslExpression() {hlslExpression=Emitter.Age *-0.25}, Maximum=HlslExpression() {hlslExpression=Emitter.Age *-0.5})
     B=-2.4
    )
   )
   &AddVelocityFromPoint(
    Module.Use Falloff Distance=false
    Velocity Strength=&RandomRangeFloat(
     Minimum=&RandomRangeFloat(Minimum=20, Maximum=26)
     Maximum=&FloatFromCurve(
      FloatCurve=DataInterface() {propertyValues {Curve {keys [{interpMode=RCIM_Cubic, tangentMode=RCTM_User, time=0.2, value=80}, {interpMode=RCIM_Cubic, tangentMode=RCTM_User, time=1, value=0}]}}}
      CurveIndex=$Emitter.Age
     )
    )
   )
   &AddVelocity(Velocity=<-200 0 0>) {enabled=false, module=AddVelocity001}
   &InheritVelocity(Speed Limit=true, UseSpeedThreshold=false)
  ]
  particleUpdateScript [
   &ParticleState()
   /Niagara/Modules/Update/Forces/CurlNoiseForce(
    Noise Frequency=0.03
    Noise Strength=&FloatFromCurve(
     FloatCurve=DataInterface() {propertyValues {Curve {keys [<at 0 0>, {arriveTangent=1, interpMode=RCIM_Cubic, leaveTangent=1, time=0.4106, value=0.6448}, <at 1 1>]}}}
     Scale Curve=300
    )
    Pan Noise=true
    Pan Noise Field=<0.5 -1 1>
   )
   &VortexForce(
    Influence Falloff Radius=150
    Use Influence Falloff=true
    UseOriginOffset=false
    Vortex Axis=<0 1 0>
    Vortex Force Amount=&FloatFromCurve(FloatCurve=DataInterface() {propertyValues {Curve {keys [<at 0 1>,<at 1 0>]}}}, Scale Curve=1500)
    Vortex Origin:NiagaraPosition=&ConvertPositionToVector(Input Position=$Engine.Owner.Position, Convert To Absolute Large World Vector=false)
   )
   &Drag(Drag=&RandomRangeFloat(Minimum=3, Maximum=7), Use Linear Drag=true, Use Rotational Drag=false)
   &ScaleColor(
    Scale Alpha=&Multiply_Float(
     A=&FloatFromCurve(FloatCurve=DataInterface() {propertyValues {Curve {keys [
          {arriveTangent=11.9433, interpMode=RCIM_Cubic, leaveTangent=11.9433, tangentMode=RCTM_User, time=0, value=0.5}
          <at 0.2 1 RCIM_Cubic>
          <at 1 0 RCIM_Cubic>
         ]}}})
     B=&Multiply_Float(A=1, B=&FNTimeOfDayScalar(Night Scale=0.5, Morning Scale=0.25, Evening Scale=0.25))
    )
    ScaleA=true
    ScaleRGB=true
   ) {module=ScaleColor001}
   &DynamicMaterialParameters(
    Index 0 Param 1=&FloatFromCurve(FloatCurve=DataInterface() {propertyValues {Curve {keys [<at 0 0>,<at 1 1>]}}})
    Index 0 Param 2=1
    Index 0 Param 3=1
    Index 0 Param 4=1
    Module.Param0WriteEnabled=true
    Module.Param1WriteEnabled=true
    Module.Param2WriteEnabled=true
    Module.Param3WriteEnabled=true
   )
   &ScaleSpriteSize(Scale Factor=&Vector2DFromFloat(Value=&FloatFromCurve(FloatCurve=DataInterface() {propertyValues {Curve {keys [
          {arriveTangent=9.1531, interpMode=RCIM_Cubic, leaveTangent=9.1531, tangentMode=RCTM_User, time=0, value=0.02}
          {arriveTangent=0.48, interpMode=RCIM_Cubic, leaveTangent=0.48, time=0.18, value=0.99}
          <at 1 0.5 RCIM_Cubic>
         ]}}})))
   &SkeletalMeshLocation(MaskSpawnGroup=false, OverrideRandomSeed=false, Skeletal Mesh=$User.SK_mesh) {enabled=false}
   &SolveForcesAndVelocity(Clamp Velocity=false, Limit Acceleration=false)
   &ScaleSpriteSizeBySpeed(Max Scale Factor=<0.5 30>, Min Scale Factor=<0.3 0.8>)
  ]
  rendererClasses [&NiagaraSpriteRendererProperties]
  renderers [{rendererClass=&NiagaraSpriteRendererProperties}]
 }
}
summary {
 systemName=NS_Mystic_Emote_r0B
 emitters [
  {emitterName=FN_Ribbon_Attributes_Test, rendererClasses [&NiagaraRibbonRendererProperties]}
  {
   emitterName=SimpleSpriteBurst
   rendererClasses [&NiagaraSpriteRendererProperties, &NiagaraLightRendererProperties]
  }
  {emitterName=Center_Glow, rendererClasses [&NiagaraSpriteRendererProperties]}
  {emitterName=sparkes, simTarget=GPUComputeSim, rendererClasses [&NiagaraSpriteRendererProperties]}
 ]
 userVariables []
}
