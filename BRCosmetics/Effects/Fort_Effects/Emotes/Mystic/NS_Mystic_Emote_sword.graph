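# Niagara effect for the Mystic emote's sword: an orange glow and sparks laid along a curved Hermite spline in local space.
# AirGlow bursts looping sprites spaced by exec index along the spline, with fade-in, sine pulse and time-of-day alpha scaling.
# Sparks_Local continuously spawns short-lived, curl-noise sparks at random spline points with tapering spread.
# User.IsSwingingMystic stops spark spawning while true.
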NiagaraSystem /BRCosmetics/Effects/Fort_Effects/Emotes/Mystic/NS_Mystic_Emote_sword
dependencies {usedDataInterfaces [&NiagaraDataInterfaceCurve], usedDynamicInputs [], usedModules [&SystemState]}
emitters {
 AirGlow {
  emitterUpdateScript [
   &EmitterState()
   &SpawnBurst_Instantaneous(Age:NiagaraFloat=$Emitter.Age, Spawn Count=8, Use Spawn Probability=false, UseLoopCountLimit=false) {module=SpawnBurst_Instantaneous001}
  ]
  particleSpawnScript [
   &InitializeParticle(
    Color=<rgba 1 0.1167 0 0.2>
    Lifetime Max=1.75
    Lifetime Min=2
    Lifetime Mode=::NewEnumerator1
    Mass Mode=::NewEnumerator3
    Sprite Rotation Mode=::NewEnumerator2
    Sprite Size Mode=::NewEnumerator3
    Uniform Sprite Size=26.7689
    UsePositionOffset=false
    Write Material Random=true
   )
   &SphereLocation(
    MaskSpawnGroup=false
    Override Seed=false
    Sphere Radius=8
    Surface Only=false
    UseNonUniformScale=false
    UseOffset=false
    UseSphereOrientation=false
   ) {enabled=false}
   &CameraOffset(Camera Offset Amount=-5)
  ]
  particleUpdateScript [
   &ParticleState(
    Kill Particles When Lifetime Has Elapsed=false
    Let Infinitely Lived Particles Die When Emitter Deactivates=true
    Loop Particles Lifetime=true
   )
   /Game/Effects/Niagara/Modules/Curves/FN_HermiteSpline(
    Compute Acceleration=true
    Compute Velocity=true
    Direction End:Vector3f=&TransformVector(Vector=<0 60 -80>, Destination Space=::Simulation)
    Direction Start:Vector3f=&TransformVector(Vector=<0 0 150>, Destination Space=::Simulation)
    Position Blend:NiagaraFloat=&FloatFromCurve(
     FloatCurve=DataInterface() {propertyValues {Curve {keys [
         {
          arriveTangent=0.3953
          arriveTangentWeight=0.4753
          interpMode=RCIM_Cubic
          leaveTangent=0.3953
          leaveTangentWeight=0.4753
          tangentMode=RCTM_User
          tangentWeightMode=RCTWM_WeightedBoth
          time=0
          value=0
         }
         {
          arriveTangent=1.2462
          arriveTangentWeight=1.0259
          interpMode=RCIM_Cubic
          leaveTangent=0.4506
          leaveTangentWeight=0.0296
          tangentMode=RCTM_Break
          tangentWeightMode=RCTWM_WeightedBoth
          time=0.95
          value=1
         }
         {interpMode=RCIM_Cubic, tangentMode=RCTM_Break, time=1, value=1}
        ]}}}
     CurveIndex=&ReturnNormalizedExecIndex()
    )
    Position End:Vector3f=/Niagara/DynamicInputs/Transforms/TransformPosition(Position=<0 -13 67>, Source Space=::Local, Destination Space=::Simulation)
    Position Start:Vector3f=&TransformPosition(Position=<0 0 -7>, Source Space=::Local, Destination Space=::Simulation)
    Write Particles Position=true
   )
   &Drag(Drag=0.25, Use Linear Drag=true, Use Rotational Drag=false)
   &ScaleColor(
    Scale Alpha=&Multiply_Float(
     A=&Lerp_Float(A=0.75, B=1, Alpha=&Sine(Normalized Angle=$Emitter.Age, Period=5, Scale=0.5, Bias=0.5))
     B=&FNTimeOfDayScalar(Night Scale=0.75, Morning Scale=1, Evening Scale=1)
    )
    ScaleA=true
    ScaleRGB=true
   )
   &ScaleColor(
    Scale Alpha=&FloatFromCurve(
     FloatCurve=DataInterface() {propertyValues {Curve {keys [
         {arriveTangent=4.1785, interpMode=RCIM_Cubic, leaveTangent=4.1785, tangentMode=RCTM_User, time=0, value=0}
         {arriveTangent=0.9995, interpMode=RCIM_Cubic, leaveTangent=0.9995, tangentMode=RCTM_User, time=0.15, value=0.6825}
         {arriveTangent=-0.0405, interpMode=RCIM_Cubic, leaveTangent=-0.0405, tangentMode=RCTM_User, time=0.4, value=1}
        ]}}}
     CurveIndex=$Emitter.Age
    )
    ScaleA=true
    ScaleRGB=false
   ) {module=ScaleColor001}
   &SolveForcesAndVelocity(Clamp Velocity=false, Limit Acceleration=false)
   &ScaleSpriteSize(
    Uniform Curve Index=&ReturnNormalizedExecIndex()
    Uniform Curve Sprite Scale=DataInterface() {propertyValues {Curve {keys [
        {
         arriveTangent=0.3512
         arriveTangentWeight=1.1064
         interpMode=RCIM_Cubic
         leaveTangent=0.3512
         leaveTangentWeight=1.1064
         tangentMode=RCTM_User
         tangentWeightMode=RCTWM_WeightedBoth
         time=0
         value=1
        }
        {
         arriveTangent=-0.9431
         arriveTangentWeight=0.1457
         interpMode=RCIM_Cubic
         leaveTangent=-0.9431
         leaveTangentWeight=0.1457
         tangentMode=RCTM_User
         tangentWeightMode=RCTWM_WeightedBoth
         time=1
         value=0.95
        }
       ]}}}
   )
  ]
  rendererClasses [&NiagaraSpriteRendererProperties]
  renderers [{rendererClass=&NiagaraSpriteRendererProperties}]
 }
 Sparks_Local {
  emitterUpdateScript [
   &EmitterState()
   &SpawnRate(
    SpawnRate=&MakeCustomFloatFromBool(Bool=$User.IsSwingingMystic, True Float=0, False Float=75)
    Use Spawn Probability=false
   )
  ]
  particleSpawnScript [
   &InitializeParticle(
    Color Maximum=<rgba 18 2.1 0 1>
    Color Minimum=<rgba 25 2.9167 0 1>
    Color Mode=::NewEnumerator2
    Lifetime Max=0.175
    Lifetime Min=0.085
    Lifetime Mode=::NewEnumerator1
    Mass Mode=::NewEnumerator3
    Sprite Rotation Mode=::NewEnumerator2
    Sprite Size Max=<2.5 3>
    Sprite Size Min=<4.5 5>
    Sprite Size Mode=::NewEnumerator2
    UsePositionOffset=false
    Write Material Random=true
   )
   &SphereLocation(
    MaskSpawnGroup=false
    Override Seed=false
    Sphere Radius=8
    Surface Only=false
    UseNonUniformScale=false
    UseOffset=false
    UseSphereOrientation=false
   ) {enabled=false}
   &AddVelocityInCone(
    Cone Angle=20
    Cone Axis=<0 0 1>
    Module.Use Velocity Falloff On Cone Axis=false
    Velocity Distribution Along Cone Axis=0.75
    Velocity Strength=&RandomRangeFloat(Minimum=500, Maximum=850)
   ) {enabled=false}
   /SetVariables(Particles.Herm Thickness:NiagaraFloat=&FloatFromCurve(
     FloatCurve=DataInterface() {propertyValues {Curve {keys [
         {arriveTangent=0.6937, interpMode=RCIM_Cubic, leaveTangent=0.6937, tangentMode=RCTM_User, time=0, value=0.075}
         {arriveTangent=-2.4361, interpMode=RCIM_Cubic, leaveTangent=-2.4361, tangentMode=RCTM_User, time=1, value=0}
        ]}}}
     CurveIndex=$Particles.MaterialRandom
    )) {isSetParameters=true}
   &FN_HermiteSpline(
    Compute Acceleration=true
    Compute Velocity=true
    Direction End:Vector3f=&TransformVector(Vector=<0 60 -80>, Destination Space=::Simulation)
    Direction Start:Vector3f=&TransformVector(Vector=<0 0 150>, Destination Space=::Simulation)
    Position Blend:NiagaraFloat=$Particles.MaterialRandom
    Position End:Vector3f=&TransformPosition(Position=<0 -13 67>, Source Space=::Local, Destination Space=::Simulation)
    Position Start:Vector3f=&TransformPosition(Position=<0 0 -7>, Source Space=::Local, Destination Space=::Simulation)
    Write Particles Position=true
   )
   &SphereLocation(
    MaskSpawnGroup=false
    Non Uniform Scale=&MakeVector(X=$Particles.Herm Thickness, Y=0.1, Z=$Particles.Herm Thickness)
    Offset=<0 0.5 0>
    Override Seed=false
    Sphere Radius=10
    Surface Only=false
    UseNonUniformScale=true
    UseOffset=true
    UseSphereOrientation=false
   ) {module=SphereLocation001}
   &AddVelocity(Velocity=&RandomVector(Vector Scale=&RandomRangeFloat(Minimum=40, Maximum=20)))
   &InheritVelocity(Inherited Velocity Amount Scale=<0.5 0.5 0.5>, Speed Limit=false, UseSpeedThreshold=false) {enabled=false}
  ]
  particleUpdateScript [
   &ParticleState()
   &GravityForce() {enabled=false}
   &Drag(Drag=0.1, Use Linear Drag=true, Use Rotational Drag=false)
   &ScaleColor(
    Scale Alpha=&FloatFromCurve(FloatCurve=DataInterface() {propertyValues {Curve {keys [<at -0.0046 1.0113 RCIM_Cubic>,<at 1.0011 0.002 RCIM_Cubic>]}}})
    ScaleA=true
    ScaleRGB=true
   )
   &CurlNoiseForce(Noise Frequency=120, Noise Quality / Cost=::NewEnumerator0, Noise Strength=5000, Pan Noise=false)
   &SolveForcesAndVelocity(Clamp Velocity=false, Limit Acceleration=false)
   &ScaleSpriteSize(Scale Factor=&Multiply_Vector2D_ByFloat(Float=1, Vector2D=<1 0.74>), Scale Sprite Size Mode=::NewEnumerator2)
  ]
  rendererClasses [&NiagaraSpriteRendererProperties]
  renderers [{rendererClass=&NiagaraSpriteRendererProperties}]
 }
}
summary {
 systemName=NS_Mystic_Emote_sword
 emitters [
  {emitterName=Sparks_Local, rendererClasses [&NiagaraSpriteRendererProperties]}
  {emitterName=AirGlow, rendererClasses [&NiagaraSpriteRendererProperties]}
 ]
 userVariables [
  User.AnimTrailColorALpha {type=LinearColor, description="", defaultValue {struct=LinearColor, value=<rgba 1 1 1 1>}}
  User.bIsFrontEnd {type=NiagaraBool}
  User.bIsFrontEndPreview {type=NiagaraBool}
  User.IsSwingingMystic {type=NiagaraBool}
  User.RibbonIDOffset {type=NiagaraInt32}
  User.SkeletalMesh {
   type=NiagaraDataInterfaceSkeletalMesh
   defaultValue {
    struct=NiagaraExt_VariableValue_DataInterface
    value {dataInterface=&self:NiagaraDataInterfaceSkeletalMesh_0}
   }
  }
 ]
}
